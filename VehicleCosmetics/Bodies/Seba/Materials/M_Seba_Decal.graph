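# Masked, Default Lit material for decals on the Seba vehicle body cosmetic.
# The graph holds a single texture sample of the Seba decals base color texture, with no connections to material outputs shown.
# It is flagged for skeletal mesh and mesh deformer use, with automatic usage flagging and ray-traced shadow casting turned off.
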Material /VehicleCosmetics/Bodies/Seba/Materials/M_Seba_Decal
graph {
 properties {
  bAutomaticallySetUsageInEditor=false
  bCastRayTracedShadows=false
  bUsedWithMeshDeformer=true
  bUsedWithSkeletalMesh=true
  blendMode=BLEND_Masked
  usedShadingModels=Default Lit
 }
 nodes [TextureSample() {texture=/VehicleCosmetics/Bodies/Seba/Textures/T_Seba_Decals_BC}]
}
