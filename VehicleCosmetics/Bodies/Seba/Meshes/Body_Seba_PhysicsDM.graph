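# Physics asset for the Seba vehicle body mesh, providing its collision shape.
# It defines one body on the root bone using a single box named root_box, offset forward and up from the bone origin.
# It has no constraints, so the vehicle body collides as one rigid piece.
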PhysicsAsset /VehicleCosmetics/Bodies/Seba/Meshes/Body_Seba_PhysicsDM
bodies {root [{shapeType=Box, extentX=444, extentY=216, extentZ=116, shapeName=root_box, center=<21.8851 0 98.75>}]}
constraints []
